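# Activatable Discovery browser screen that plays an activity's video with audio on a black background, with a Skip button bound to GenericBack.
# PlayVideo and PlayVideo_LinkEntryVM start the embedded ActivityVideoCycleWidget from their respective view models.
# The link-entry variant also plays a no-rating bumper when the Frontend.EnableVideoNoRating console variable is true.
# Video end, bound on Construct, deactivates the screen; Skip stops the video and calls the same handler.
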WidgetBlueprint /DiscoveryBrowser/ActivityVideoScreen
graphs [
 EventGraph {nodes [
   "On Clicked (Button_Skip)"!() -> (OutputDelegate, then, Button:Common Button Base Object Reference) {is=title K2Node_ComponentBoundEvent}
   > StopVideo()
   n1 = OnVideoEnded() {is=title K2Node_CustomEvent}
   > DeactivateWidget()
   Event Construct()
   > Bind Event to Video Ended(
    self:Activity Video Cycle Widget Object Reference=Get ActivityVideoCycleWidget() -> (ActivityVideoCycleWidget:Activity Video Cycle Widget Object Reference)
    Delegate=@n1
   ) {is=title K2Node_AddDelegate}
  ]}
 PlayVideo {nodes [
   n1 = PlayVideo!() -> (then, ActivityViewModel:Discover Activity View Model Object Reference) {is=title K2Node_FunctionEntry}
   > StartVideoCycleFromViewModel!(
    execute
    self:Fort Activity Video Cycle Object Reference=Get ActivityVideoCycleWidget()
    ActivityViewModel:Discover Activity View Model Object Reference=@n1
   ) -> (then)
  ]}
 PlayVideo_LinkEntryVM {nodes [
   n1 = PlayVideo_LinkEntryVM!() -> (then, LinkEntryViewModel:Fort Discover Link Entry Tile Item VM Object Reference) {is=title K2Node_FunctionEntry}
   > StartVideoCycleFromLinkEntryViewModel!(
    execute
    self:Fort Activity Video Cycle Object Reference=Get ActivityVideoCycleWidget()
    LinkEntryViewModel:Fort Discover Link Entry Tile Item VM Object Reference=@n1
   ) -> (then)
   > n2 = GetConsoleVariableBoolValue!(execute, VariableName:String=Frontend.EnableVideoNoRating) -> (then, ReturnValue:bool)
   > Branch(Condition=@n2)
   > PlayVideoBumperNoRating(self:Activity Video Cycle Widget Object Reference=Get ActivityVideoCycleWidget())
  ]}
 StopVideo {nodes [
   StopVideo() {is=title K2Node_FunctionEntry}
   > StopVideo(self:Fort Activity Video Cycle Object Reference=Get ActivityVideoCycleWidget()) {is=title K2Node_CallFunction}
   > OnVideoEnded() {is=title K2Node_CallFunction}
  ]}
]
variables []
widgets {
 info {inheritedWidgetCount=0, parentClass=&CommonActivatableWidget, rootWidgetClass=&Border, widgetCount=5}
 widgets [
  SafezonePadding {
   widgetClassPath=&Border
   widget=SafezonePadding
   properties {visibility=SelfHitTestInvisible, brushColor=<rgba 0 0 0 1>, padding=<ltrb 0 0 0 0>}
  }
  Overlay_0 {widgetClassPath=&Overlay, parent=SafezonePadding, widget=Overlay_0, slot {class=BorderSlot, properties {padding=<ltrb 0 0 0 0>}}}
  ActivityVideoCycleWidget {
   widgetClassPath=/FortniteCoreUI/UI/DiscoveryShared/GameDetails/ActivityVideoCycleWidget
   bIsVariable=true
   parent=Overlay_0
   widget=ActivityVideoCycleWidget
   properties {
    activeSoundMix=/Game/Sounds/FortSoundClassesAndModes/Fort_Cine_DuckIngame
    bHasAudio=true
    bUseFullscreenDenylistCheck=true
    soundClass=/Game/Sounds/FortSoundClassesAndModes/RenderedCinematics
    videoMaterial=/DiscoveryBrowser/Tiles/Islands/Materials/MI_UI_IslandTile_Video
   }
   slot {class=OverlaySlot, properties {horizontalAlignment=HAlign_Fill, verticalAlignment=VAlign_Fill}}
  }
  SafeZone_SkipButton {
   widgetClassPath=&SafeZone
   bIsVariable=true
   parent=Overlay_0
   widget=SafeZone_SkipButton
   properties {padLeft=false, padTop=false}
   slot {class=OverlaySlot, properties {horizontalAlignment=HAlign_Right, verticalAlignment=VAlign_Bottom, padding=<ltrb 0 0 32 32>}}
  }
  Button_Skip {
   widgetClassPath=/FortUILibrary/UIKit/Buttons/WBP_UIKit_Button_Quiet
   bIsVariable=true
   parent=SafeZone_SkipButton
   widget=Button_Skip
   properties {
    bHideInputActionWithKeyboard=true
    holdCompletedSound=/Game/Sounds/Fort_UI_Sounds/FrontEnd/New_FrontEnd/DefaultSounds/HoldableButtons/Small_Quiet_Clear/RenderedCine/UI_RenderedCine_H_oldable_Small_Complete_Cue_old
    holdPressedSound=/Game/Sounds/Fort_UI_Sounds/FrontEnd/New_FrontEnd/DefaultSounds/HoldableButtons/Small_Quiet_Clear/RenderedCine/UI_RenderedCine_H_oldable_Small_Pressed_Cue_old
    holdReleasedSound=/Game/Sounds/Fort_UI_Sounds/FrontEnd/New_FrontEnd/DefaultSounds/HoldableButtons/Small_Quiet_Clear/RenderedCine/UI_RenderedCine_H_oldable_Small_Release_Cue_old
    holdSound=/Game/Sounds/Fort_UI_Sounds/FrontEnd/New_FrontEnd/DefaultSounds/HoldableButtons/Small_Quiet_Clear/RenderedCine/UI_RenderedCine_H_oldable_Small_H_old_Cue_old
    text=Skip
    hoveredSlateSoundOverride {resourceObject=/Game/Sounds/Fort_UI_Sounds/FrontEnd/SC_UI_Cinematics_SkipButton_Hover}
    pressedSlateSoundOverride {resourceObject=/Game/Sounds/Fort_UI_Sounds/FrontEnd/SC_UI_Cinematics_SkipButton_Pressed}
    triggeringInputAction {dataTable=&InputActionDataTable, rowName=GenericBack}
   }
   slot {class=SafeZoneSlot, properties {hAlign=HAlign_Left, vAlign=VAlign_Top}}
  }
 ]
}
