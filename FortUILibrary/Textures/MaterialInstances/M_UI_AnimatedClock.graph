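# Translucent UI material that draws an animated clock icon procedurally, without textures, tinted by an Icon Color parameter.
# A filled circle forms the face, and a center dot plus two box-shaped hands are cut out of it through a lerp to zero.
# Time and a negated Speed parameter drive UV rotators for each hand, with the shorter hand turning at one sixtieth of the longer hand's rate.
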Material /FortUILibrary/Textures/MaterialInstances/M_UI_AnimatedClock
graph {
 properties {
  bAutomaticallySetUsageInEditor=false
  bUsedWithStaticMesh=false
  blendMode=BLEND_Translucent
  materialDomain=MD_UI
  usedShadingModels=Default Lit
 }
 nodes [
  n1 = Time()
  n2 = NamedRerouteUsage()
  n3 = MaterialFunctionCall(
   UVs=MaterialFunctionCall() {materialFunction=&GetUserInterfaceUV}.Normalized UV
   "Rotation Angle (0-1)"=Frac(None=Multiply(A=@n1, B=@n2) {constB=-0.1})
  ) {materialFunction=&CustomRotator}
  n4 = Subtract(A=ComponentMask(None=@n3.Rotated Values) {g=true}) {constB=-0.1406}
  n5 = ComponentMask(None=@n3.Rotated Values) {r=true}
  n6 = Add(
   A=MaterialFunctionCall(Size=Constant() {r=0.125}) {materialFunction=&MF_UI_Circle}.Fill
   B=MaterialFunctionCall(UVs=AppendVector(A=@n5, B=@n4), Size=Constant2Vector() {g=0.2812, r=0.125}) {materialFunction=&MF_UI_Box}.Fill
  )
  n7 = Time()
  n8 = Divide(A=NamedRerouteUsage()) {constB=60}
  n9 = MaterialFunctionCall(
   UVs=MaterialFunctionCall() {materialFunction=&GetUserInterfaceUV}.Normalized UV
   "Rotation Angle (0-1)"=Frac(None=Multiply(A=@n7, B=@n8) {constB=-0.0017})
  ) {materialFunction=&CustomRotator}
  n10 = ComponentMask(None=@n9.Rotated Values) {r=true}
  n11 = MaterialFunctionCall(
   UVs=AppendVector(A=Subtract(A=@n10) {constB=0.0938}, B=ComponentMask(None=@n9.Rotated Values) {g=true})
   Size=Constant2Vector() {g=0.125, r=0.1875}
  ) {materialFunction=&MF_UI_Box}
  LinearInterpolate(
   A=MaterialFunctionCall(Size=Constant() {r=0.9375}) {materialFunction=&MF_UI_Circle}.Fill
   Alpha=Saturate(None=Add(A=@n6, B=@n11.Fill))
  ) {constB=0}
  NamedRerouteDeclaration(None=VectorParameter() {parameterName=Icon Color, defaultValue=<rgba 1 1 1 1>}.RGB) {name=Final Color, nodeColor=<rgba 0.6875 1 0 1>}
  NamedRerouteDeclaration(None=Multiply(A=ScalarParameter() {defaultValue=0.1, parameterName=Speed}) {constB=-1}) {name=Speed, nodeColor=<rgba 0.9922 0 1 1>}
  NamedRerouteUsage()
 ]
}
